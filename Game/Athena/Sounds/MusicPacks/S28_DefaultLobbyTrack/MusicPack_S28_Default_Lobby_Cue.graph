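# Season 28 default lobby music pack cue that plays the lobby login-loop track as non-spatialized music.
# A looping wave player feeds an Enveloper whose volume curve stays near silent briefly, then ramps up to full, delaying and fading in the start.
# The cue uses the music packs sound class, limits itself to one instance by stopping the oldest, and sends to the music audio-analysis submix.
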SoundCue /Game/Athena/Sounds/MusicPacks/S28_DefaultLobbyTrack/MusicPack_S28_Default_Lobby_Cue
root=Enveloper_0
nodes [
 {
  id=Enveloper_0
  children [WavePlayer_2]
  properties {
   pitchCurve {editorCurveData {keys [{time=0}]}}
   volumeCurve {editorCurveData {keys [
      <at 0 0 RCIM_Cubic>
      <at 0.4987 -0.0001>
      {arriveTangent=4.1409, interpMode=RCIM_Cubic, leaveTangent=4.1409, tangentMode=RCTM_User, time=1, value=1.0336}
     ]}}
  }
 }
 {
  id=WavePlayer_2
  properties {bLooping=true, soundWaveAssetPtr=&/S28_LobbyTrack_Default_LoginLoop_v2}
 }
]
properties {
 bOverrideConcurrency=true
 maxDistance=2097152
 soundClassObject=/Game/Sounds/FortSoundClassesAndModes/Fort_Music_MusicPacks
 volumeMultiplier=1
 attenuationOverrides {bAttenuate=false, bSpatialize=false}
 concurrencyOverrides {maxCount=1, resolutionRule=StopOldest}
 soundSubmixSends [{minSendLevel=0, soundSubmix=&Fortnite_AudioAnalysis_Music}]
}
